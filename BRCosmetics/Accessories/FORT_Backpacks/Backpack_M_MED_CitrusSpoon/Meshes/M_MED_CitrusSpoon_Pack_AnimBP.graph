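# Animation Blueprint for the CitrusSpoon backpack cosmetic that makes its dangling wires swing with physics.
# The event graph samples the body mesh's spine socket transforms and computes a gravity override; world-space bone modifiers on the spine feed a cached pose.
# Two stacks of five AnimDynamics chains simulate center, left, right and top wires, with limits against spine, pack and neighboring wires.
# Wearer gender selects the stack, differing in spine collision radius.
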AnimBlueprint /BRCosmetics/Accessories/FORT_Backpacks/Backpack_M_MED_CitrusSpoon/Meshes/M_MED_CitrusSpoon_Pack_AnimBP
target_skeleton=&/M_MED_CitrusSpoon_Pack_Skeleton
graphs [
 AnimGraph {nodes [
   n1 = "Anim Dynamics - Chain: dyn_wire_center_01 -> dyn_wire_center_04"!(ComponentPose=Local To Component(LocalPose=Use cached pose 'start pose'()), Alpha=1, GravityOverride:Vector="0,0,-980") -> (Pose) {
    angularDampingOverride=1
    angularSpringConstant=500
    bChain=true
    bOverrideAngularDamping=true
    bUseGravityOverride=true
    bUseSphericalLimits=true
    numSolverIterationsPostUpdate=4
    numSolverIterationsPreUpdate=16
    simulationSpace=RootRelative
    alphaBoolBlend {blendOption=Linear}
    boundBone {boneName=dyn_wire_center_01}
    chainEnd {boneName=dyn_wire_center_04}
    componentAppliedLinearAccClamp=<1000 1000 1000>
    componentLinearAccScale=<0.4 0.4 0.4>
    componentLinearVelScale=<0.4 0.4 0.4>
    physicsBodyDefinitions [
     {
      collisionType=CustomSphere
      sphereCollisionRadius=1
      boundBone {boneName=dyn_wire_center_01}
      boxExtents=<23 23 23>
      constraintSetup {angularLimitsMax=<0 45 15>, angularLimitsMin=<0 -45 -15>, angularTarget=<1 0 0>}
      localJointOffset=<1 0 0>
     }
     {
      collisionType=CustomSphere
      sphereCollisionRadius=1
      boundBone {boneName=dyn_wire_center_02}
      boxExtents=<17 17 17>
      constraintSetup {angularLimitsMax=<0 30 20>, angularLimitsMin=<0 -30 -20>, angularTarget=<1 0 0>}
      localJointOffset=<1 0 0>
     }
     {
      collisionType=CustomSphere
      sphereCollisionRadius=1
      boundBone {boneName=dyn_wire_center_03}
      boxExtents=<12 12 12>
      constraintSetup {angularLimitsMax=<0 20 25>, angularLimitsMin=<0 -20 -25>, angularTarget=<1 0 0>}
      localJointOffset=<1 0 0>
     }
     {
      collisionType=CustomSphere
      sphereCollisionRadius=1
      boundBone {boneName=dyn_wire_center_04}
      boxExtents=<6 6 6>
      constraintSetup {angularLimitsMax=<0 15 35>, angularLimitsMin=<0 -15 -35>, angularTarget=<1 0 0>}
      localJointOffset=<1 0 0>
     }
    ]
    planarLimits [{drivingBone {boneName=pack}, planeTransform {location=<0 2 0>, rotation=<rot 0 0 -90>}}]
    sphericalLimits [
     {limitRadius=12, drivingBone {boneName=spine_01}, sphereLocalOffset=<0 0 -5>}
     {limitRadius=12, drivingBone {boneName=spine_02}, sphereLocalOffset=<0 0 -5>}
     {limitRadius=12, drivingBone {boneName=spine_03}, sphereLocalOffset=<0 0 -5>}
     {limitRadius=1, drivingBone {boneName=dyn_wire_01_L}, sphereLocalOffset=<2 0 0>}
     {limitRadius=1, drivingBone {boneName=dyn_wire_02_L}, sphereLocalOffset=<1.5 0 0>}
     {limitRadius=1, drivingBone {boneName=dyn_wire_03_L}, sphereLocalOffset=<1 0 0>}
     {limitRadius=1, drivingBone {boneName=dyn_wire_01_R}, sphereLocalOffset=<2 0 0>}
     {limitRadius=1, drivingBone {boneName=dyn_wire_02_R}, sphereLocalOffset=<2 0 0>}
     {limitRadius=1, drivingBone {boneName=dyn_wire_02_R}, sphereLocalOffset=<4 0 0>}
     {limitRadius=1, drivingBone {boneName=dyn_wire_02_R}, sphereLocalOffset=<0 0 0>}
     {limitRadius=1, drivingBone {boneName=dyn_wire_03_R}, sphereLocalOffset=<1.5 0 0>}
     {limitRadius=1, drivingBone {boneName=dyn_wire_02_L}, sphereLocalOffset=<-0.5 0 0>}
     {limitRadius=1, drivingBone {boneName=dyn_wire_02_L}, sphereLocalOffset=<3.5 0 0>}
     {limitRadius=1, drivingBone {boneName=dyn_wire_03_L}, sphereLocalOffset=<3 0 0>}
    ]
   }
   n2 = "Anim Dynamics - Chain: dyn_wire_01_L -> dyn_wire_03_L"!(ComponentPose=@n1, Alpha=1, GravityOverride:Vector="0,0,-980") -> (Pose) {
    angularDampingOverride=1
    angularSpringConstant=500
    bChain=true
    bOverrideAngularDamping=true
    bUseGravityOverride=true
    bUseSphericalLimits=true
    numSolverIterationsPostUpdate=4
    numSolverIterationsPreUpdate=16
    simulationSpace=RootRelative
    alphaBoolBlend {blendOption=Linear}
    boundBone {boneName=dyn_wire_01_L}
    chainEnd {boneName=dyn_wire_03_L}
    componentAppliedLinearAccClamp=<1000 1000 1000>
    componentLinearAccScale=<0.45 0.45 0.45>
    componentLinearVelScale=<0.45 0.45 0.45>
    physicsBodyDefinitions [
     {
      collisionType=CustomSphere
      sphereCollisionRadius=1
      boundBone {boneName=dyn_wire_01_L}
      boxExtents=<20 20 20>
      constraintSetup {angularLimitsMax=<0 55 15>, angularLimitsMin=<0 -55 -15>, angularTarget=<1 0 0>}
      localJointOffset=<1 0 0>
     }
     {
      collisionType=CustomSphere
      sphereCollisionRadius=1
      boundBone {boneName=dyn_wire_02_L}
      boxExtents=<15 15 15>
      constraintSetup {angularLimitsMax=<0 35 20>, angularLimitsMin=<0 -35 -20>, angularTarget=<1 0 0>}
      localJointOffset=<1 0 0>
     }
     {
      collisionType=CustomSphere
      sphereCollisionRadius=1
      boundBone {boneName=dyn_wire_03_L}
      boxExtents=<7 7 7>
      constraintSetup {angularLimitsMax=<0 25 25>, angularLimitsMin=<0 -25 -25>, angularTarget=<1 0 0>}
      localJointOffset=<1 0 0>
     }
    ]
    planarLimits [{drivingBone {boneName=pack}, planeTransform {location=<0 0 0>, rotation=<rot 0 0 -90>}}]
    sphericalLimits [
     {limitRadius=12, drivingBone {boneName=spine_01}, sphereLocalOffset=<0 0 -5>}
     {limitRadius=12, drivingBone {boneName=spine_02}, sphereLocalOffset=<0 0 -5>}
     {limitRadius=12, drivingBone {boneName=spine_03}, sphereLocalOffset=<0 0 -5>}
     {limitRadius=1, drivingBone {boneName=dyn_wire_center_01}, sphereLocalOffset=<0 0 0>}
     {limitRadius=1, drivingBone {boneName=dyn_wire_center_02}, sphereLocalOffset=<0 0 0>}
     {limitRadius=1, drivingBone {boneName=dyn_wire_center_03}, sphereLocalOffset=<0 0 0>}
     {limitRadius=1, drivingBone {boneName=dyn_wire_center_04}, sphereLocalOffset=<0 0 0>}
     {limitRadius=1, drivingBone {boneName=dyn_wire_center_02}, sphereLocalOffset=<-2.1 0 0>}
     {limitRadius=1, drivingBone {boneName=dyn_wire_center_02}, sphereLocalOffset=<2.1 0 0>}
     {limitRadius=1, drivingBone {boneName=dyn_wire_center_04}, sphereLocalOffset=<-2.1 0 0>}
     {limitRadius=1, drivingBone {boneName=dyn_wire_01_R}, sphereLocalOffset=<2 0 0>}
     {limitRadius=1, drivingBone {boneName=dyn_wire_02_R}, sphereLocalOffset=<2 0 0>}
     {limitRadius=1, drivingBone {boneName=dyn_wire_02_R}, sphereLocalOffset=<4 0 0>}
     {limitRadius=1, drivingBone {boneName=dyn_wire_02_R}, sphereLocalOffset=<0 0 0>}
     {limitRadius=1, drivingBone {boneName=dyn_wire_03_R}, sphereLocalOffset=<1.5 0 0>}
    ]
   }
   n3 = "Anim Dynamics - Chain: dyn_wire_01_R -> dyn_wire_03_R"!(ComponentPose=@n2, Alpha=1, GravityOverride:Vector="0,0,-980") -> (Pose) {
    angularDampingOverride=1
    angularSpringConstant=500
    bChain=true
    bOverrideAngularDamping=true
    bUseGravityOverride=true
    bUseSphericalLimits=true
    numSolverIterationsPostUpdate=4
    numSolverIterationsPreUpdate=16
    simulationSpace=RootRelative
    alphaBoolBlend {blendOption=Linear}
    boundBone {boneName=dyn_wire_01_R}
    chainEnd {boneName=dyn_wire_03_R}
    componentAppliedLinearAccClamp=<1000 1000 1000>
    componentLinearAccScale=<0.5 0.5 0.5>
    componentLinearVelScale=<0.5 0.5 0.5>
    physicsBodyDefinitions [
     {
      collisionType=CustomSphere
      sphereCollisionRadius=1
      boundBone {boneName=dyn_wire_01_R}
      boxExtents=<16 16 16>
      constraintSetup {angularLimitsMax=<0 55 20>, angularLimitsMin=<0 -55 -20>, angularTarget=<1 0 0>}
      localJointOffset=<1 0 0>
     }
     {
      collisionType=CustomSphere
      sphereCollisionRadius=1
      boundBone {boneName=dyn_wire_02_R}
      boxExtents=<11 11 11>
      constraintSetup {angularLimitsMax=<0 35 25>, angularLimitsMin=<0 -35 -25>, angularTarget=<1 0 0>}
      localJointOffset=<1 0 0>
     }
     {
      collisionType=CustomSphere
      sphereCollisionRadius=1
      boundBone {boneName=dyn_wire_03_R}
      boxExtents=<6 6 6>
      constraintSetup {angularLimitsMax=<0 25 30>, angularLimitsMin=<0 -25 -30>, angularTarget=<1 0 0>}
      localJointOffset=<1 0 0>
     }
    ]
    planarLimits [{drivingBone {boneName=pack}, planeTransform {location=<0 0 0>, rotation=<rot 0 0 -90>}}]
    sphericalLimits [
     {limitRadius=12, drivingBone {boneName=spine_01}, sphereLocalOffset=<0 0 -5>}
     {limitRadius=12, drivingBone {boneName=spine_02}, sphereLocalOffset=<0 0 -5>}
     {limitRadius=12, drivingBone {boneName=spine_03}, sphereLocalOffset=<0 0 -5>}
     {limitRadius=1, drivingBone {boneName=dyn_wire_01_L}, sphereLocalOffset=<2 0 0>}
     {limitRadius=1, drivingBone {boneName=dyn_wire_02_L}, sphereLocalOffset=<1.5 0 0>}
     {limitRadius=1, drivingBone {boneName=dyn_wire_03_L}, sphereLocalOffset=<1 0 0>}
     {limitRadius=1, drivingBone {boneName=dyn_wire_02_L}, sphereLocalOffset=<-0.5 0 0>}
     {limitRadius=1, drivingBone {boneName=dyn_wire_02_L}, sphereLocalOffset=<3.5 0 0>}
     {limitRadius=1, drivingBone {boneName=dyn_wire_03_L}, sphereLocalOffset=<3 0 0>}
     {limitRadius=1, drivingBone {boneName=dyn_wire_center_01}, sphereLocalOffset=<0 0 0>}
     {limitRadius=1, drivingBone {boneName=dyn_wire_center_02}, sphereLocalOffset=<0 0 0>}
     {limitRadius=1, drivingBone {boneName=dyn_wire_center_03}, sphereLocalOffset=<0 0 0>}
     {limitRadius=1, drivingBone {boneName=dyn_wire_center_04}, sphereLocalOffset=<0 0 0>}
     {limitRadius=1, drivingBone {boneName=dyn_wire_center_02}, sphereLocalOffset=<-2.1 0 0>}
     {limitRadius=1, drivingBone {boneName=dyn_wire_center_02}, sphereLocalOffset=<2.1 0 0>}
     {limitRadius=1, drivingBone {boneName=dyn_wire_center_04}, sphereLocalOffset=<-2.1 0 0>}
    ]
   }
   n4 = "Anim Dynamics - Chain: dyn_wire_top_fr_01 -> dyn_wire_top_fr_03"!(ComponentPose=@n3, Alpha=1, GravityOverride:Vector="0,0,-980") -> (Pose) {
    angularDampingOverride=1
    angularSpringConstant=50000
    bChain=true
    bOverrideAngularDamping=true
    bUseGravityOverride=true
    numSolverIterationsPostUpdate=4
    numSolverIterationsPreUpdate=16
    simulationSpace=RootRelative
    alphaBoolBlend {blendOption=Linear}
    boundBone {boneName=dyn_wire_top_fr_01}
    chainEnd {boneName=dyn_wire_top_fr_03}
    componentAppliedLinearAccClamp=<1000 1000 1000>
    componentLinearAccScale=<0.4 0.4 0.5>
    componentLinearVelScale=<0.4 0.4 0.5>
    physicsBodyDefinitions [
     {
      boundBone {boneName=dyn_wire_top_fr_01}
      boxExtents=<13 13 13>
      constraintSetup {angularLimitsMax=<0 0 15>, angularLimitsMin=<0 -45 0>, angularTarget=<1 0 0>}
      localJointOffset=<1 0 0>
     }
     {
      boundBone {boneName=dyn_wire_top_fr_02}
      boxExtents=<9 9 9>
      constraintSetup {angularLimitsMax=<0 0 15>, angularLimitsMin=<0 -30 0>, angularTarget=<1 0 0>}
      localJointOffset=<1 0 0>
     }
     {
      boundBone {boneName=dyn_wire_top_fr_03}
      boxExtents=<5 5 5>
      constraintSetup {angularLimitsMax=<0 0 15>, angularLimitsMin=<0 -20 0>, angularTarget=<1 0 0>}
      localJointOffset=<-2.7 3 0>
     }
    ]
   }
   n5 = "Anim Dynamics - Chain: dyn_wire_top_bk_01 -> dyn_wire_top_bk_03"!(ComponentPose=@n4, Alpha=1, GravityOverride:Vector="0,0,-980") -> (Pose) {
    angularDampingOverride=1
    angularSpringConstant=1000
    bChain=true
    bOverrideAngularDamping=true
    bUseGravityOverride=true
    numSolverIterationsPostUpdate=4
    numSolverIterationsPreUpdate=16
    simulationSpace=RootRelative
    alphaBoolBlend {blendOption=Linear}
    boundBone {boneName=dyn_wire_top_bk_01}
    chainEnd {boneName=dyn_wire_top_bk_03}
    componentAppliedLinearAccClamp=<1000 1000 1000>
    componentLinearAccScale=<0.4 0.4 0.5>
    componentLinearVelScale=<0.4 0.4 0.5>
    physicsBodyDefinitions [
     {
      boundBone {boneName=dyn_wire_top_bk_01}
      boxExtents=<16 16 16>
      constraintSetup {angularLimitsMax=<0 10 10>, angularLimitsMin=<0 0 0>, angularTarget=<1 0 0>}
      localJointOffset=<1 0 0>
     }
     {
      boundBone {boneName=dyn_wire_top_bk_02}
      boxExtents=<11 11 11>
      constraintSetup {angularLimitsMax=<0 35 15>, angularLimitsMin=<0 0 0>, angularTarget=<1 0 0>}
      localJointOffset=<1 0 0>
     }
     {
      boundBone {boneName=dyn_wire_top_bk_03}
      boxExtents=<6 6 6>
      constraintSetup {angularLimitsMax=<0 35 15>, angularLimitsMin=<0 0 0>, angularTarget=<1 0 0>}
      localJointOffset=<-3 0 0>
     }
    ]
   }
   n6 = "Anim Dynamics - Chain: dyn_wire_center_01 -> dyn_wire_center_04"(ComponentPose=Local To Component(LocalPose=Use cached pose 'start pose'()), Alpha=1, GravityOverride="0,0,-980") {
    angularDampingOverride=1
    angularSpringConstant=500
    bChain=true
    bOverrideAngularDamping=true
    bUseGravityOverride=true
    bUseSphericalLimits=true
    numSolverIterationsPostUpdate=4
    numSolverIterationsPreUpdate=16
    simulationSpace=RootRelative
    alphaBoolBlend {blendOption=Linear}
    boundBone {boneName=dyn_wire_center_01}
    chainEnd {boneName=dyn_wire_center_04}
    componentAppliedLinearAccClamp=<1000 1000 1000>
    componentLinearAccScale=<0.4 0.4 0.4>
    componentLinearVelScale=<0.4 0.4 0.4>
    physicsBodyDefinitions [
     {
      collisionType=CustomSphere
      sphereCollisionRadius=1
      boundBone {boneName=dyn_wire_center_01}
      boxExtents=<23 23 23>
      constraintSetup {angularLimitsMax=<0 45 15>, angularLimitsMin=<0 -45 -15>, angularTarget=<1 0 0>}
      localJointOffset=<1 0 0>
     }
     {
      collisionType=CustomSphere
      sphereCollisionRadius=1
      boundBone {boneName=dyn_wire_center_02}
      boxExtents=<17 17 17>
      constraintSetup {angularLimitsMax=<0 30 20>, angularLimitsMin=<0 -30 -20>, angularTarget=<1 0 0>}
      localJointOffset=<1 0 0>
     }
     {
      collisionType=CustomSphere
      sphereCollisionRadius=1
      boundBone {boneName=dyn_wire_center_03}
      boxExtents=<12 12 12>
      constraintSetup {angularLimitsMax=<0 20 25>, angularLimitsMin=<0 -20 -25>, angularTarget=<1 0 0>}
      localJointOffset=<1 0 0>
     }
     {
      collisionType=CustomSphere
      sphereCollisionRadius=1
      boundBone {boneName=dyn_wire_center_04}
      boxExtents=<6 6 6>
      constraintSetup {angularLimitsMax=<0 15 35>, angularLimitsMin=<0 -15 -35>, angularTarget=<1 0 0>}
      localJointOffset=<1 0 0>
     }
    ]
    planarLimits [{drivingBone {boneName=pack}, planeTransform {location=<0 2 0>, rotation=<rot 0 0 -90>}}]
    sphericalLimits [
     {limitRadius=14, drivingBone {boneName=spine_01}, sphereLocalOffset=<0 0 -5>}
     {limitRadius=14, drivingBone {boneName=spine_02}, sphereLocalOffset=<0 0 -5>}
     {limitRadius=14, drivingBone {boneName=spine_03}, sphereLocalOffset=<0 0 -5>}
     {limitRadius=1, drivingBone {boneName=dyn_wire_01_L}, sphereLocalOffset=<2 0 0>}
     {limitRadius=1, drivingBone {boneName=dyn_wire_02_L}, sphereLocalOffset=<1.5 0 0>}
     {limitRadius=1, drivingBone {boneName=dyn_wire_03_L}, sphereLocalOffset=<1 0 0>}
     {limitRadius=1, drivingBone {boneName=dyn_wire_01_R}, sphereLocalOffset=<2 0 0>}
     {limitRadius=1, drivingBone {boneName=dyn_wire_02_R}, sphereLocalOffset=<2 0 0>}
     {limitRadius=1, drivingBone {boneName=dyn_wire_02_R}, sphereLocalOffset=<4 0 0>}
     {limitRadius=1, drivingBone {boneName=dyn_wire_02_R}, sphereLocalOffset=<0 0 0>}
     {limitRadius=1, drivingBone {boneName=dyn_wire_03_R}, sphereLocalOffset=<1.5 0 0>}
     {limitRadius=1, drivingBone {boneName=dyn_wire_02_L}, sphereLocalOffset=<-0.5 0 0>}
     {limitRadius=1, drivingBone {boneName=dyn_wire_02_L}, sphereLocalOffset=<3.5 0 0>}
     {limitRadius=1, drivingBone {boneName=dyn_wire_03_L}, sphereLocalOffset=<3 0 0>}
    ]
   }
   n7 = "Anim Dynamics - Chain: dyn_wire_01_L -> dyn_wire_03_L"(ComponentPose=@n6, Alpha=1, GravityOverride="0,0,-980") {
    angularDampingOverride=1
    angularSpringConstant=500
    bChain=true
    bOverrideAngularDamping=true
    bUseGravityOverride=true
    bUseSphericalLimits=true
    numSolverIterationsPostUpdate=4
    numSolverIterationsPreUpdate=16
    simulationSpace=RootRelative
    alphaBoolBlend {blendOption=Linear}
    boundBone {boneName=dyn_wire_01_L}
    chainEnd {boneName=dyn_wire_03_L}
    componentAppliedLinearAccClamp=<1000 1000 1000>
    componentLinearAccScale=<0.45 0.45 0.45>
    componentLinearVelScale=<0.45 0.45 0.45>
    physicsBodyDefinitions [
     {
      collisionType=CustomSphere
      sphereCollisionRadius=1
      boundBone {boneName=dyn_wire_01_L}
      boxExtents=<20 20 20>
      constraintSetup {angularLimitsMax=<0 55 15>, angularLimitsMin=<0 -55 -15>, angularTarget=<1 0 0>}
      localJointOffset=<1 0 0>
     }
     {
      collisionType=CustomSphere
      sphereCollisionRadius=1
      boundBone {boneName=dyn_wire_02_L}
      boxExtents=<15 15 15>
      constraintSetup {angularLimitsMax=<0 35 20>, angularLimitsMin=<0 -35 -20>, angularTarget=<1 0 0>}
      localJointOffset=<1 0 0>
     }
     {
      collisionType=CustomSphere
      sphereCollisionRadius=1
      boundBone {boneName=dyn_wire_03_L}
      boxExtents=<7 7 7>
      constraintSetup {angularLimitsMax=<0 25 25>, angularLimitsMin=<0 -25 -25>, angularTarget=<1 0 0>}
      localJointOffset=<1 0 0>
     }
    ]
    planarLimits [{drivingBone {boneName=pack}, planeTransform {location=<0 0 0>, rotation=<rot 0 0 -90>}}]
    sphericalLimits [
     {limitRadius=14, drivingBone {boneName=spine_01}, sphereLocalOffset=<0 0 -5>}
     {limitRadius=14, drivingBone {boneName=spine_02}, sphereLocalOffset=<0 0 -5>}
     {limitRadius=14, drivingBone {boneName=spine_03}, sphereLocalOffset=<0 0 -5>}
     {limitRadius=1, drivingBone {boneName=dyn_wire_center_01}, sphereLocalOffset=<0 0 0>}
     {limitRadius=1, drivingBone {boneName=dyn_wire_center_02}, sphereLocalOffset=<0 0 0>}
     {limitRadius=1, drivingBone {boneName=dyn_wire_center_03}, sphereLocalOffset=<0 0 0>}
     {limitRadius=1, drivingBone {boneName=dyn_wire_center_04}, sphereLocalOffset=<0 0 0>}
     {limitRadius=1, drivingBone {boneName=dyn_wire_center_02}, sphereLocalOffset=<-2.1 0 0>}
     {limitRadius=1, drivingBone {boneName=dyn_wire_center_02}, sphereLocalOffset=<2.1 0 0>}
     {limitRadius=1, drivingBone {boneName=dyn_wire_center_04}, sphereLocalOffset=<-2.1 0 0>}
     {limitRadius=1, drivingBone {boneName=dyn_wire_01_R}, sphereLocalOffset=<2 0 0>}
     {limitRadius=1, drivingBone {boneName=dyn_wire_02_R}, sphereLocalOffset=<2 0 0>}
     {limitRadius=1, drivingBone {boneName=dyn_wire_02_R}, sphereLocalOffset=<4 0 0>}
     {limitRadius=1, drivingBone {boneName=dyn_wire_02_R}, sphereLocalOffset=<0 0 0>}
     {limitRadius=1, drivingBone {boneName=dyn_wire_03_R}, sphereLocalOffset=<1.5 0 0>}
    ]
   }
   n8 = "Anim Dynamics - Chain: dyn_wire_01_R -> dyn_wire_03_R"(ComponentPose=@n7, Alpha=1, GravityOverride="0,0,-980") {
    angularDampingOverride=1
    angularSpringConstant=500
    bChain=true
    bOverrideAngularDamping=true
    bUseGravityOverride=true
    bUseSphericalLimits=true
    numSolverIterationsPostUpdate=4
    numSolverIterationsPreUpdate=16
    simulationSpace=RootRelative
    alphaBoolBlend {blendOption=Linear}
    boundBone {boneName=dyn_wire_01_R}
    chainEnd {boneName=dyn_wire_03_R}
    componentAppliedLinearAccClamp=<1000 1000 1000>
    componentLinearAccScale=<0.5 0.5 0.5>
    componentLinearVelScale=<0.5 0.5 0.5>
    physicsBodyDefinitions [
     {
      collisionType=CustomSphere
      sphereCollisionRadius=1
      boundBone {boneName=dyn_wire_01_R}
      boxExtents=<16 16 16>
      constraintSetup {angularLimitsMax=<0 55 20>, angularLimitsMin=<0 -55 -20>, angularTarget=<1 0 0>}
      localJointOffset=<1 0 0>
     }
     {
      collisionType=CustomSphere
      sphereCollisionRadius=1
      boundBone {boneName=dyn_wire_02_R}
      boxExtents=<11 11 11>
      constraintSetup {angularLimitsMax=<0 35 25>, angularLimitsMin=<0 -35 -25>, angularTarget=<1 0 0>}
      localJointOffset=<1 0 0>
     }
     {
      collisionType=CustomSphere
      sphereCollisionRadius=1
      boundBone {boneName=dyn_wire_03_R}
      boxExtents=<6 6 6>
      constraintSetup {angularLimitsMax=<0 25 30>, angularLimitsMin=<0 -25 -30>, angularTarget=<1 0 0>}
      localJointOffset=<1 0 0>
     }
    ]
    planarLimits [{drivingBone {boneName=pack}, planeTransform {location=<0 0 0>, rotation=<rot 0 0 -90>}}]
    sphericalLimits [
     {limitRadius=14, drivingBone {boneName=spine_01}, sphereLocalOffset=<0 0 -5>}
     {limitRadius=14, drivingBone {boneName=spine_02}, sphereLocalOffset=<0 0 -5>}
     {limitRadius=14, drivingBone {boneName=spine_03}, sphereLocalOffset=<0 0 -5>}
     {limitRadius=1, drivingBone {boneName=dyn_wire_01_L}, sphereLocalOffset=<2 0 0>}
     {limitRadius=1, drivingBone {boneName=dyn_wire_02_L}, sphereLocalOffset=<1.5 0 0>}
     {limitRadius=1, drivingBone {boneName=dyn_wire_03_L}, sphereLocalOffset=<1 0 0>}
     {limitRadius=1, drivingBone {boneName=dyn_wire_02_L}, sphereLocalOffset=<-0.5 0 0>}
     {limitRadius=1, drivingBone {boneName=dyn_wire_02_L}, sphereLocalOffset=<3.5 0 0>}
     {limitRadius=1, drivingBone {boneName=dyn_wire_03_L}, sphereLocalOffset=<3 0 0>}
     {limitRadius=1, drivingBone {boneName=dyn_wire_center_01}, sphereLocalOffset=<0 0 0>}
     {limitRadius=1, drivingBone {boneName=dyn_wire_center_02}, sphereLocalOffset=<0 0 0>}
     {limitRadius=1, drivingBone {boneName=dyn_wire_center_03}, sphereLocalOffset=<0 0 0>}
     {limitRadius=1, drivingBone {boneName=dyn_wire_center_04}, sphereLocalOffset=<0 0 0>}
     {limitRadius=1, drivingBone {boneName=dyn_wire_center_02}, sphereLocalOffset=<-2.1 0 0>}
     {limitRadius=1, drivingBone {boneName=dyn_wire_center_02}, sphereLocalOffset=<2.1 0 0>}
     {limitRadius=1, drivingBone {boneName=dyn_wire_center_04}, sphereLocalOffset=<-2.1 0 0>}
    ]
   }
   n9 = "Anim Dynamics - Chain: dyn_wire_top_fr_01 -> dyn_wire_top_fr_03"(ComponentPose=@n8, Alpha=1, GravityOverride="0,0,-980") {
    angularDampingOverride=1
    angularSpringConstant=50000
    bChain=true
    bOverrideAngularDamping=true
    bUseGravityOverride=true
    numSolverIterationsPostUpdate=4
    numSolverIterationsPreUpdate=16
    simulationSpace=RootRelative
    alphaBoolBlend {blendOption=Linear}
    boundBone {boneName=dyn_wire_top_fr_01}
    chainEnd {boneName=dyn_wire_top_fr_03}
    componentAppliedLinearAccClamp=<1000 1000 1000>
    componentLinearAccScale=<0.4 0.4 0.5>
    componentLinearVelScale=<0.4 0.4 0.5>
    physicsBodyDefinitions [
     {
      boundBone {boneName=dyn_wire_top_fr_01}
      boxExtents=<13 13 13>
      constraintSetup {angularLimitsMax=<0 0 15>, angularLimitsMin=<0 -45 0>, angularTarget=<1 0 0>}
      localJointOffset=<1 0 0>
     }
     {
      boundBone {boneName=dyn_wire_top_fr_02}
      boxExtents=<9 9 9>
      constraintSetup {angularLimitsMax=<0 0 15>, angularLimitsMin=<0 -30 0>, angularTarget=<1 0 0>}
      localJointOffset=<1 0 0>
     }
     {
      boundBone {boneName=dyn_wire_top_fr_03}
      boxExtents=<5 5 5>
      constraintSetup {angularLimitsMax=<0 0 15>, angularLimitsMin=<0 -20 0>, angularTarget=<1 0 0>}
      localJointOffset=<-2.7 3 0>
     }
    ]
   }
   n10 = "Anim Dynamics - Chain: dyn_wire_top_bk_01 -> dyn_wire_top_bk_03"(ComponentPose=@n9, Alpha=1, GravityOverride="0,0,-980") {
    angularDampingOverride=1
    angularSpringConstant=1000
    bChain=true
    bOverrideAngularDamping=true
    bUseGravityOverride=true
    numSolverIterationsPostUpdate=4
    numSolverIterationsPreUpdate=16
    simulationSpace=RootRelative
    alphaBoolBlend {blendOption=Linear}
    boundBone {boneName=dyn_wire_top_bk_01}
    chainEnd {boneName=dyn_wire_top_bk_03}
    componentAppliedLinearAccClamp=<1000 1000 1000>
    componentLinearAccScale=<0.4 0.4 0.5>
    componentLinearVelScale=<0.4 0.4 0.5>
    physicsBodyDefinitions [
     {
      boundBone {boneName=dyn_wire_top_bk_01}
      boxExtents=<16 16 16>
      constraintSetup {angularLimitsMax=<0 10 10>, angularLimitsMin=<0 0 0>, angularTarget=<1 0 0>}
      localJointOffset=<1 0 0>
     }
     {
      boundBone {boneName=dyn_wire_top_bk_02}
      boxExtents=<11 11 11>
      constraintSetup {angularLimitsMax=<0 35 15>, angularLimitsMin=<0 0 0>, angularTarget=<1 0 0>}
      localJointOffset=<1 0 0>
     }
     {
      boundBone {boneName=dyn_wire_top_bk_03}
      boxExtents=<6 6 6>
      constraintSetup {angularLimitsMax=<0 35 15>, angularLimitsMin=<0 0 0>, angularTarget=<1 0 0>}
      localJointOffset=<-3 0 0>
     }
    ]
   }
   Output Pose(Result=Blend Poses by bool(
     BlendPose_0=Component To Local(ComponentPose=@n10)
     BlendPose_1=Component To Local(ComponentPose=@n5)
     bActiveValue=Get isMale() -> (isMale:bool)
    ) {blendType=Linear}) {layerGroup=None}
   n11 = "Transform (Modify) Bone - Bone: spine_01"!(ComponentPose=Local To Component(LocalPose=Input Pose()), Alpha=1, Translation="0,0,0", Rotation="0,0,0") -> (Pose) {
    rotationMode=BMM_Replace
    rotationSpace=BCS_WorldSpace
    translationMode=BMM_Replace
    translationSpace=BCS_WorldSpace
    alphaBoolBlend {blendOption=Linear}
    boneToModify {boneName=spine_01}
   }
   n12 = "Transform (Modify) Bone - Bone: spine_02"!(ComponentPose=@n11, Alpha=1, Translation="0,0,0", Rotation="0,0,0") -> (Pose) {
    rotationMode=BMM_Replace
    rotationSpace=BCS_WorldSpace
    translationMode=BMM_Replace
    translationSpace=BCS_WorldSpace
    alphaBoolBlend {blendOption=Linear}
    boneToModify {boneName=spine_02}
   }
   n13 = "Transform (Modify) Bone - Bone: spine_03"!(ComponentPose=@n12, Alpha=1, Translation="0,0,0", Rotation="0,0,0") -> (Pose) {
    rotationMode=BMM_Replace
    rotationSpace=BCS_WorldSpace
    translationMode=BMM_Replace
    translationSpace=BCS_WorldSpace
    alphaBoolBlend {blendOption=Linear}
    boneToModify {boneName=spine_03}
   }
   Save cached pose 'start pose'(Pose=Component To Local(ComponentPose=@n13)) {cacheName=start pose}
  ]}
 EventGraph {nodes [
   Event BlueprintUpdateAnimation()
   > n1 = Sequence()
   > n2 = Gravity Override State Function DEPRECATED(CharacterPartAnimInstance=Self-Reference(), GameTimeInSeconds=GetGameTimeInSeconds(), GravityOverrideParams=Get Gravity Override Params())
   > Set Gravity Override(Gravity Override=@n2)
   n3 = Get BodyMesh() -> (BodyMesh:Skeletal Mesh Component Object Reference)
   Is Valid(exec=@n1.then_1, InputObject=@n3)
   n4 = BreakTransform(InTransform=GetSocketTransform(self=@n3, InSocketName=spine_01, TransformSpace=RTS_World))
   > Set spine 01 translation(spine 01 translation:Vector=@n4) -> (Output_Get:Vector)
   > Set spine 01 rotation(spine 01 rotation:Rotator=@n4.Rotation) -> (Output_Get:Rotator)
   n5 = BreakTransform(InTransform=GetSocketTransform(self=@n3, InSocketName=spine_02, TransformSpace=RTS_World))
   > Set spine 02 translation(spine 02 translation:Vector=@n5) -> (Output_Get:Vector)
   > Set spine 02 rotation(spine 02 rotation:Rotator=@n5.Rotation) -> (Output_Get:Rotator)
   n6 = BreakTransform(InTransform=GetSocketTransform(self=@n3, InSocketName=spine_03, TransformSpace=RTS_World))
   > Set spine 03 translation(spine 03 translation:Vector=@n6) -> (Output_Get:Vector)
   > Set spine 03 rotation(spine 03 rotation:Rotator=@n6.Rotation) -> (Output_Get:Rotator)
   n7 = Event OnCharacterCustomizationLoaded()
   > Set BodyMesh(BodyMesh=Get Array Item with EFort Custom Part Type(Array=@n7.CustomMeshParts))
   > n8 = Cast To PlayerPawn_Athena(Object=TryGetPawnOwner())
   > n9 = FindGender!(execute, Object:Object Reference=@n8) -> (True:Exec, False:Exec, OutGender:EFortCustomGender Enum=Invalid)
   > Set isMale(isMale:bool="Equal (Enum)"(A:EFortCustomGender Enum=@n9, B:EFortCustomGender Enum=Male)) -> (Output_Get:bool)
   TryGetPawnOwner()
  ]}
]
variables [
 BodyMesh {type=SkeletalMeshComponent}
 Gravity Override {type=Vector}
 Gravity Override Params {type=GravityOverrideParamsStruct}
 isMale {type=boolean}
 spine 01 rotation {type=Rotator}
 spine 01 translation {type=Vector}
 spine 02 rotation {type=Rotator}
 spine 02 translation {type=Vector}
 spine 03 rotation {type=Rotator}
 spine 03 translation {type=Vector}
]
